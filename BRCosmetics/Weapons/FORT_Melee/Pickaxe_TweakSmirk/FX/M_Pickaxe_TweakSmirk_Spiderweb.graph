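# Translucent, unlit, two-sided material for spiderweb effects on the TweakSmirk pickaxe, built for Niagara mesh particles.
# SplineThicken drives a vertex offset blendable toward a constant, with base and tip widths optionally scaled by a dynamic parameter.
# Particle color feeds emissive adjustments; opacity combines particle alpha, a powered texture mask and view-angle fade.
# A shading-path switch applies depth fade on default and deferred paths, skipping it on forward and mobile.
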Material /BRCosmetics/Weapons/FORT_Melee/Pickaxe_TweakSmirk/FX/M_Pickaxe_TweakSmirk_Spiderweb
graph {
 properties {
  bAutomaticallySetUsageInEditor=false
  bUsedWithNiagaraMeshParticles=true
  bUsedWithStaticMesh=false
  blendMode=BLEND_Translucent
  pixelDepthOffsetMode=PDOM_Legacy
  shadingModel=MSM_Unlit
  twoSided=true
  usedShadingModels=Unlit
 }
 nodes [
  n1 = ScalarParameter() {defaultValue=3, parameterName=SplineBase}
  n2 = DynamicParameter() {
   parameterIndex=3
   defaultValue=<rgba 1 2 1 200>
   paramNames [Noise_Intensity_offset, FloorNoise_SpeedOffset, CustomSplineWidth, WorldPosition_SphereMaskRadius]
  }
  n3 = LinearInterpolate(B=@n2.CustomSplineWidth, Alpha=ScalarParameter() {parameterName=Use Custom SplineWidth}) {constA=1}
  n4 = ScalarParameter() {defaultValue=3, parameterName=SplineTip}
  LinearInterpolate(
   A=MaterialFunctionCall(WidthBase=Multiply(A=@n1, B=@n3), WidthTip=Multiply(A=@n4, B=@n3), Expand U or V UV Channel=StaticBool()) {materialFunction=&SplineThicken}.WorldPositionOffset
   B=Constant()
   Alpha=ScalarParameter() {parameterName=Use_SplineThicken}
  )
  n5 = ParticleColor()
  MaterialFunctionCall(Emissive=@n5.RGB) {materialFunction=&MF_FX_EmissiveAdjustments}
  n6 = ScalarParameter() {defaultValue=1, parameterName=Alpha}
  n7 = TextureSampleParameter2D(UVs=TextureCoordinate()) {group=MaskTexture, parameterName=MaskTexture, texture=/Game/Weapons/FORT_Rifles/Effects/Textures/T_FX_ThinLine_BoomB_01}
  n8 = ScalarParameter() {defaultValue=1, parameterName=Power}
  n9 = Multiply(
   A=Multiply(A=Multiply(A=@n6, B=@n5.A), B=Power(Base=@n7.R, Exp=@n8))
   B=MaterialFunctionCall("FadeSensitivity (optional)"=ScalarParameter() {parameterName=MeshFade}) {materialFunction=&FadeBasedOnViewAngle}.Result
  )
  n10 = DepthFade(Opacity=@n9, FadeDistance=ScalarParameter() {defaultValue=5, parameterName=DepthFade})
  ShadingPathSwitch(Default=@n10, Deferred=@n10, Forward=@n9, Mobile=@n9)
 ]
}
