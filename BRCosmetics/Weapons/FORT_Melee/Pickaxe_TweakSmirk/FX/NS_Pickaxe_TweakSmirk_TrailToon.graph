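# Niagara swing-trail effect for the TweakSmirk pickaxe, drawing a short-lived, HDR hot-pink ribbon.
# Its single emitter uses anim-trail modules to spawn, initialize and update the ribbon from the system's Skeletal Mesh.
# Alpha fades over particle life, multiplied by a system-level SystemFade.
# The system also depends on time-of-day select and float-switch inputs, and exposes front-end, ribbon ID and swing-section user parameters.
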NiagaraSystem /BRCosmetics/Weapons/FORT_Melee/Pickaxe_TweakSmirk/FX/NS_Pickaxe_TweakSmirk_TrailToon
dependencies {
 usedDataInterfaces [&NiagaraDataInterfaceCurve, &NiagaraDataInterfaceSkeletalMesh]
 usedDynamicInputs [
  &MergeHelper_FloatSwitch
  &FN_Select_Float_from_Time_Of_Day
 ]
 usedModules [
  &self:SystemSpawnScript.SystemScriptSource.SystemScriptGraph.NiagaraNodeAssignment_0.SetVariables_#1
  &SystemState
 ]
}
emitters {TrailHighlight002 {
  emitterUpdateScript [
   &EmitterState()
   &FN_SpawnAnimTrail(AnimTrailWidthMode=::FromSecond, Switch on ENiagara_AnimTrail_Source:ENiagara_AnimTrail_Source=::NewEnumerator1)
  ]
  particleSpawnScript [
   &FN_InitializeAnimTrailBoneIndecies()
   &FN_InitializeAnimTrail(AnimTrailWidth=1.7, Color=<rgba 10 0.3 3.0483 2>, Lifetime=0.3, SkeletalMesh=$System.Skeletal Mesh)
  ]
  particleUpdateScript [
   &ParticleState()
   &FN_UpdateAnimTrail(SkeletalMesh=$System.Skeletal Mesh)
   &ScaleColor(
    Scale Alpha=&Multiply_Float(
     A=&FloatFromCurve(FloatCurve=DataInterface() {propertyValues {Curve {keys [<at 0 1>,<at 1 0>]}}})
     B=$System.SystemFade
    )
    Scale RGB=&VectorFromFloat(Value=1)
    ScaleA=true
    ScaleRGB=true
   )
   &ScaleRibbonWidth(Ribbon Width Scale=1.14)
  ]
  rendererClasses [&NiagaraRibbonRendererProperties]
  renderers [{rendererClass=&NiagaraRibbonRendererProperties}]
 }}
summary {
 systemName=NS_Pickaxe_TweakSmirk_TrailToon
 emitters [{emitterName=TrailHighlight002, rendererClasses [&NiagaraRibbonRendererProperties]}]
 userVariables [
  User.bIsFrontEnd {type=NiagaraBool}
  User.bIsFrontEndPreview {type=NiagaraBool}
  User.FrontEndScale {type=NiagaraFloat, defaultValue=0.6, description=""}
  User.RibbonIDOffset {type=NiagaraInt32}
  User.SwingSectionIndex {type=NiagaraInt32}
 ]
}
