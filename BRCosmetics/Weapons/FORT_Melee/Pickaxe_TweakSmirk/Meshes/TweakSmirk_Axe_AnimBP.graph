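# Animation Blueprint for the TweakSmirk pickaxe that adds physical sway to its three-bone dangling string chain.
# The cached input pose feeds three chained AnimDynamics variants for dyn_string_01 to 03, each with gravity, angular limits and high damping.
# Nested bool blends select the angular-spring variant outside the front end, and one of two springless tunings inside it, depending on whether the item is being previewed.
# The EventGraph holds only unwired stub nodes.
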AnimBlueprint /BRCosmetics/Weapons/FORT_Melee/Pickaxe_TweakSmirk/Meshes/TweakSmirk_Axe_AnimBP
target_skeleton=&/TweakSmirk_Axe_Skeleton
graphs [
 AnimGraph {nodes [
   n1 = Anim Dynamics - Bone: dyn_string_01(ComponentPose=Local To Component(LocalPose=Use cached pose 'InitialPose'()), Alpha=1, ExternalForce="0,0,0") {
    angularDampingOverride=0.999
    bOverrideAngularDamping=true
    bUseGravityOverride=true
    boundBone {boneName=dyn_string_01}
    componentAppliedLinearAccClamp=<1000 1000 5000>
    componentLinearAccScale=<0.3 0.3 0.3>
    componentLinearVelScale=<0.3 0.3 0.3>
    gravityOverride {z=-980}
    physicsBodyDefinitions [{
      boundBone {boneName=dyn_string_01}
      boxExtents=<10 10 10>
      constraintSetup {angularLimitsMax=<0 75 75>, angularLimitsMin=<0 -75 -75>, angularTarget=<0 0 0>}
      localJointOffset=<1.7 0 0>
     }]
    simSpaceSettings {maxAngularAcceleration=8000, maxAngularVelocity=8000, simSpaceAngularAlpha=0.18}
   }
   n2 = Anim Dynamics - Bone: dyn_string_02(ComponentPose=@n1, Alpha=1, ExternalForce="0,0,0") {
    angularDampingOverride=0.995
    bOverrideAngularDamping=true
    bUseGravityOverride=true
    boundBone {boneName=dyn_string_02}
    componentAppliedLinearAccClamp=<1000 1000 5000>
    componentLinearAccScale=<0.4 0.4 0.4>
    componentLinearVelScale=<0.4 0.4 0.4>
    gravityOverride {z=-980}
    physicsBodyDefinitions [{
      boundBone {boneName=dyn_string_02}
      boxExtents=<10 10 10>
      constraintSetup {angularLimitsMax=<0 90 90>, angularLimitsMin=<0 -90 -90>, angularTarget=<0 0 0>}
      localJointOffset=<1.7 0 0>
     }]
    simSpaceSettings {maxAngularAcceleration=8000, maxAngularVelocity=8000, simSpaceAngularAlpha=0.18}
   }
   n3 = Anim Dynamics - Bone: dyn_string_03(ComponentPose=@n2, Alpha=1, ExternalForce="0,0,0") {
    angularDampingOverride=0.99
    bOverrideAngularDamping=true
    bUseGravityOverride=true
    boundBone {boneName=dyn_string_03}
    componentAppliedLinearAccClamp=<5000 5000 5000>
    componentLinearAccScale=<0.4 0.4 0.4>
    componentLinearVelScale=<0.4 0.4 0.4>
    gravityOverride {z=-980}
    physicsBodyDefinitions [{
      boundBone {boneName=dyn_string_03}
      boxExtents=<10 10 10>
      constraintSetup {angularLimitsMax=<45 45 45>, angularLimitsMin=<-45 -45 -45>, angularTarget=<0 0 0>}
      localJointOffset=<1.7 0 0>
     }]
    simSpaceSettings {maxAngularAcceleration=8000, maxAngularVelocity=8000, simSpaceAngularAlpha=0.18}
   }
   n4 = Component To Local(ComponentPose=@n3)
   n5 = Anim Dynamics - Bone: dyn_string_01(ComponentPose=Local To Component(LocalPose=Use cached pose 'InitialPose'()), Alpha=1, ExternalForce="0,0,0") {
    angularDampingOverride=0.9999
    bOverrideAngularDamping=true
    bUseGravityOverride=true
    boundBone {boneName=dyn_string_01}
    componentAppliedLinearAccClamp=<1000 1000 5000>
    componentLinearAccScale=<0.3 0.3 0.3>
    componentLinearVelScale=<0.3 0.3 0.3>
    gravityOverride {z=-980}
    physicsBodyDefinitions [{
      boundBone {boneName=dyn_string_01}
      boxExtents=<15 10 10>
      constraintSetup {angularLimitsMax=<0 75 75>, angularLimitsMin=<0 -75 -75>, angularTarget=<0 0 0>}
      localJointOffset=<1.4 0 0>
     }]
   }
   n6 = Anim Dynamics - Bone: dyn_string_02(ComponentPose=@n5, Alpha=1, ExternalForce="0,0,0") {
    angularDampingOverride=0.9995
    bOverrideAngularDamping=true
    bUseGravityOverride=true
    boundBone {boneName=dyn_string_02}
    componentAppliedLinearAccClamp=<1000 1000 5000>
    componentLinearAccScale=<0.4 0.4 0.4>
    componentLinearVelScale=<0.4 0.4 0.4>
    gravityOverride {z=-980}
    physicsBodyDefinitions [{
      boundBone {boneName=dyn_string_02}
      boxExtents=<15 10 10>
      constraintSetup {angularLimitsMax=<0 90 90>, angularLimitsMin=<0 -90 -90>, angularTarget=<0 0 0>}
      localJointOffset=<1.4 0 0>
     }]
   }
   n7 = Anim Dynamics - Bone: dyn_string_03(ComponentPose=@n6, Alpha=1, ExternalForce="0,0,0") {
    angularDampingOverride=0.999
    bOverrideAngularDamping=true
    bUseGravityOverride=true
    boundBone {boneName=dyn_string_03}
    componentAppliedLinearAccClamp=<5000 5000 5000>
    componentLinearAccScale=<0.4 0.4 0.4>
    componentLinearVelScale=<0.4 0.4 0.4>
    gravityOverride {z=-980}
    physicsBodyDefinitions [{
      boundBone {boneName=dyn_string_03}
      boxExtents=<15 10 10>
      constraintSetup {angularLimitsMax=<45 45 45>, angularLimitsMin=<-45 -45 -45>, angularTarget=<0 0 0>}
      localJointOffset=<1.4 0 0>
     }]
   }
   n8 = Component To Local(ComponentPose=@n7)
   n9 = Get bIsPreviewingItemInFrontEnd() -> (bIsPreviewingItemInFrontEnd:bool)
   n10 = Anim Dynamics - Bone: dyn_string_01(ComponentPose=Local To Component(LocalPose=Use cached pose 'InitialPose'()), Alpha=1, ExternalForce="0,0,0") {
    angularDampingOverride=0.999
    angularSpringConstant=100
    bOverrideAngularDamping=true
    bUseGravityOverride=true
    boundBone {boneName=dyn_string_01}
    componentAppliedLinearAccClamp=<1000 1000 5000>
    componentLinearAccScale=<0.3 0.3 0.3>
    componentLinearVelScale=<0.3 0.3 0.3>
    gravityOverride {z=-980}
    physicsBodyDefinitions [{
      boundBone {boneName=dyn_string_01}
      boxExtents=<10 10 10>
      constraintSetup {angularLimitsMax=<0 75 75>, angularLimitsMin=<0 -75 -75>, angularTarget=<0 0 0>}
      localJointOffset=<1.7 0 0>
     }]
   }
   n11 = Anim Dynamics - Bone: dyn_string_02(ComponentPose=@n10, Alpha=1, ExternalForce="0,0,0") {
    angularDampingOverride=0.995
    angularSpringConstant=100
    bAngularSpring=true
    bOverrideAngularDamping=true
    bUseGravityOverride=true
    boundBone {boneName=dyn_string_02}
    componentAppliedLinearAccClamp=<1000 1000 5000>
    componentLinearAccScale=<0.4 0.4 0.4>
    componentLinearVelScale=<0.4 0.4 0.4>
    gravityOverride {z=-980}
    physicsBodyDefinitions [{
      boundBone {boneName=dyn_string_02}
      boxExtents=<10 10 10>
      constraintSetup {angularLimitsMax=<0 90 90>, angularLimitsMin=<0 -90 -90>, angularTarget=<0 0 0>}
      localJointOffset=<1.7 0 0>
     }]
   }
   n12 = Anim Dynamics - Bone: dyn_string_03(ComponentPose=@n11, Alpha=1, ExternalForce="0,0,0") {
    angularDampingOverride=0.99
    angularSpringConstant=100
    bAngularSpring=true
    bOverrideAngularDamping=true
    bUseGravityOverride=true
    boundBone {boneName=dyn_string_03}
    componentAppliedLinearAccClamp=<5000 5000 5000>
    componentLinearAccScale=<0.4 0.4 0.4>
    componentLinearVelScale=<0.4 0.4 0.4>
    gravityOverride {z=-980}
    physicsBodyDefinitions [{
      boundBone {boneName=dyn_string_03}
      boxExtents=<10 10 10>
      constraintSetup {angularLimitsMax=<45 45 45>, angularLimitsMin=<-45 -45 -45>, angularTarget=<0 0 0>}
      localJointOffset=<1.7 0 0>
     }]
   }
   Output Pose(Result=Blend Poses by bool(
     BlendPose_0=Blend Poses by bool(BlendPose_0=@n8, BlendPose_1=@n4, bActiveValue=@n9)
     BlendPose_1=Component To Local(ComponentPose=@n12)
     bActiveValue=Get bIsInFrontEnd()
    )) {layerGroup=None}
   AnimGraphNode_SubInput_0 = Input Pose()
   Save cached pose 'InitialPose'(Pose=@AnimGraphNode_SubInput_0) {cacheName=InitialPose}
  ]}
 EventGraph {nodes [
   TryGetPawnOwner()
   Event BlueprintUpdateAnimation()
  ]}
]
variables []
